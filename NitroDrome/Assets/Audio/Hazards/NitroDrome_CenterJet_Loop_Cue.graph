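# Looping sound cue for the NitroDrome center jet hazard, in the traps sound class with its own attenuation asset.
# A looping wave feeds an Enveloper that quickly fades volume in, then a smoothed Doppler node.
# The root quality node routes that chain to its first two slots and leaves the other two empty, silencing those levels.
# A concurrency override allows one instance per owner and blocks new ones.
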SoundCue /NitroDrome/Assets/Audio/Hazards/NitroDrome_CenterJet_Loop_Cue
root=QualityLevel_0
nodes [
 {id=Doppler_0, children [Enveloper_0], properties {bUseSmoothing=true, dopplerIntensity=0.5}}
 {
  id=Enveloper_0
  children [WavePlayer_0]
  properties {pitchCurve {editorCurveData {keys [{time=0}]}}, volumeCurve {editorCurveData {keys [<at 0 0>,<at 0.15 1>]}}}
 }
 {id=QualityLevel_0, children [Doppler_0,Doppler_0,null,null]}
 {id=WavePlayer_0, properties {bLooping=true, soundWaveAssetPtr=&/NitroDrome_CenterJet_Loop}}
]
properties {
 attenuationSettings=&/NitroDome_CenterJet_Loop_Att
 bOverrideConcurrency=true
 maxDistance=1500
 soundClassObject=&Fort_Traps
 volumeMultiplier=4
 concurrencyOverrides {bLimitToOwner=true, maxCount=1, resolutionRule=PreventNew}
}
